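# Niagara system for the MediCrow emote hand effects: purple auras and wispy trails on the player's skeletal mesh.
# Both emitters use the partition index to pick a bone or socket, with per-side offsets or forces and orientation from its rotation.
# Aura bursts two sprites that follow the mesh, pop in and fade.
# WispyTrail continuously spawns ribbon particles that drift downward and outward, fading in and out by emitter age.
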NiagaraSystem /BRCosmetics/Animation/Game/MainPlayer/Emotes/MediCrow/FX/NS_Emote_MediCrow_Hands
dependencies {
 usedDataInterfaces [
  &NiagaraDataInterfaceCurve
  &NiagaraDataInterfaceColorCurve
  &NiagaraDataInterfaceVector2DCurve
  &NiagaraDataInterfaceSkeletalMesh
 ]
 usedDynamicInputs [
  &RandomRangeFloat
  &Multiply_Vector2D_ByFloat
  &RandomRangeVector2D
  &FNTimeOfDayColor
 ]
 usedModules [
  &self:SystemSpawnScript.SystemScriptSource.SystemScriptGraph.NiagaraNodeAssignment_0.SetVariables_#1
  &SystemState
  &self:SystemSpawnScript.SystemScriptSource.SystemScriptGraph.NiagaraNodeAssignment_1.SetVariables_#2
  &self:SystemSpawnScript.SystemScriptSource.SystemScriptGraph.NiagaraNodeAssignment_2.SetVariables_#3
 ]
}
emitters {
 Aura {
  emitterUpdateScript [
   &EmitterState(Life Cycle Mode=::NewEnumerator1, Loop Behavior=::NewEnumerator1, Loop Duration=2.4, UseLoopDelay=false)
   &SpawnBurst_Instantaneous(Spawn Count=2, Use Spawn Probability=false, UseLoopCountLimit=false)
  ]
  particleSpawnScript [
   &PartitionParticles()
   &InitializeParticle(
    Lifetime=2.4
    Sprite Rotation Mode=::NewEnumerator1
    Sprite Size Mode=::NewEnumerator3
    Sprite UV Mode=::NewEnumerator1
    Uniform Sprite Size=40
    UsePositionOffset=false
    Write Material Random=true
   )
   &CameraOffset(Camera Offset Amount=5)
   /SetVariables(Particles.PivotOffset:Vector2f=/Niagara/DynamicInputs/Helpers/MaskVector2d(
     Mask Parameter:NiagaraInt32=$Particles.PartitionParticles.Partition
     Mask:NiagaraInt32=1
     Masked Vector2D:Vector2f=<0.5 0.4>
     Passthrough Vector2D:Vector2f=<0.5 0.6>
    )) {isSetParameters=true}
  ]
  particleUpdateScript [
   &SkeletalMeshLocation(
    Bone / Socket Index=$Particles.PartitionParticles.Partition
    Bone Sampling Mode=::NewEnumerator4
    MaskSpawnGroup=false
    Orientation Sampling=::NewEnumerator1
    Skeletal Mesh=$System.Skeletal Mesh
   )
   &SpriteFacingAndAlignment(
    Sprite Alignment=&RotateVectorByQuaternion(Vector To Rotate=<1 0 0>, Quaternion=$Particles.SkeletalMeshLocation.SampledRotation)
    WriteAlignment=true
    WriteFacing=false
   )
   &ParticleState()
   &ScaleSpriteSize(
    Non-Uniform Curve Scale=<1.8 1.8>
    Non-Uniform Curve Sprite Scale=DataInterface() {propertyValues {
      XCurve {keys [{interpMode=RCIM_Cubic, tangentMode=RCTM_User, time=0, value=0}, {interpMode=RCIM_Cubic, tangentMode=RCTM_User, time=0.05, value=0.6085}]}
      YCurve {keys [{interpMode=RCIM_Cubic, tangentMode=RCTM_User, time=0, value=0}, {interpMode=RCIM_Cubic, tangentMode=RCTM_User, time=0.05, value=0.9144}]}
     }}
    Scale Sprite Size Mode=::NewEnumerator3
   ) {module=ScaleSpriteSize001}
   &ScaleColor(
    Scale Alpha=&FloatFromCurve(FloatCurve=DataInterface() {propertyValues {Curve {keys [
         <at 0 0 RCIM_Cubic>
         {arriveTangent=1.0349, interpMode=RCIM_Cubic, leaveTangent=1.0349, time=0.1424, value=1}
         {interpMode=RCIM_Cubic, tangentMode=RCTM_User, time=0.5993, value=0.6202}
         {interpMode=RCIM_Cubic, tangentMode=RCTM_User, time=1, value=0}
        ]}}})
    Scale RGB=&MakeVectorFromLinearColorRGB(Color=&ColorFromCurve(ColorCurve=DataInterface() {propertyValues {
        AlphaCurve {keys [<at 0 1 RCIM_Cubic>]}
        BlueCurve {keys [<at 0 1 RCIM_Cubic>,<at 0.1921 1 RCIM_Cubic>]}
        GreenCurve {keys [<at 0 0.1751 RCIM_Cubic>,<at 0.1921 0.0904 RCIM_Cubic>]}
        RedCurve {keys [<at 0 0.7898 RCIM_Cubic>,<at 0.1921 0.4323 RCIM_Cubic>]}
       }}))
    ScaleA=true
    ScaleRGB=true
   )
  ]
  rendererClasses [&NiagaraSpriteRendererProperties]
  renderers [{rendererClass=&NiagaraSpriteRendererProperties}]
 }
 WispyTrail {
  emitterUpdateScript [
   &EmitterState(Life Cycle Mode=::NewEnumerator1, Loop Behavior=::NewEnumerator1, Loop Delay=0.3, Loop Duration=2, UseLoopDelay=false)
   &SpawnRate(SpawnRate=40, Use Spawn Probability=false)
  ]
  particleSpawnScript [
   &PartitionParticles()
   &InitializeParticle(UsePositionOffset=false, Write Material Random=true)
   &SkeletalMeshLocation(
    Bone / Socket Index=$Particles.PartitionParticles.Partition
    Bone Sampling Mode=::NewEnumerator4
    MaskSpawnGroup=false
    Orientation Sampling=::NewEnumerator1
    Skeletal Mesh=$System.Skeletal Mesh
   )
   &InitializeRibbon(
    Color=<rgba 0.3254 0 1 3>
    Lifetime=&FloatFromCurve(FloatCurve=DataInterface() {propertyValues {Curve {keys [<at 0 1>,<at 1 0>]}}})
    Position Mode:ENiagara_PositionInitializationMode=::NewEnumerator0
    Ribbon Facing Mode:ENiagara_UnsetDirectSet=::NewEnumerator1
    Ribbon Facing Vector:Vector3f=&RotateVectorByQuaternion(Vector To Rotate=<1 0 0>, Quaternion=$Particles.SkeletalMeshLocation.SampledRotation)
    Ribbon Width=80
    Write Material Random=true
   )
   &AddVelocity(Velocity=<0 0 -50>)
   &PointForce(
    Force Strength=70
    Linear Force Vector:Vector3f=/Niagara/DynamicInputs/Helpers/MaskPosition(
     Mask Parameter:NiagaraInt32=$Particles.PartitionParticles.Partition
     Mask:NiagaraInt32=0
     Masked Position:NiagaraPosition=<1 0 0>
     Passthrough Position:NiagaraPosition=<-1 0 0>
    )
    Module.Use Falloff=false
    Module.UseLinearForce=true
   )
   &ApplyInitialForces()
  ]
  particleUpdateScript [
   &ParticleState()
   &SolveForcesAndVelocity(Clamp Velocity=false, Limit Acceleration=false)
   &DynamicMaterialParameters(
    Index 0 Param 1=1
    Module.Index0Float0Write=true
    Module.Index0Float1Write=false
    Module.Index0Float2Write=true
    Module.Index0Float3Write=true
   ) {module=DynamicMaterialParameters001}
   &ScaleColor(
    Scale Alpha=&FloatFromCurve(
     FloatCurve=DataInterface() {propertyValues {Curve {keys [
         {arriveTangent=3.8691, interpMode=RCIM_Cubic, leaveTangent=3.8691, tangentMode=RCTM_User, time=0.1437, value=0}
         {arriveTangent=1.1918, interpMode=RCIM_Cubic, leaveTangent=1.1918, time=0.3847, value=0.877}
         {arriveTangent=0.2554, interpMode=RCIM_Cubic, leaveTangent=0.2554, tangentMode=RCTM_User, time=0.9828, value=1}
         {arriveTangent=-0.4506, interpMode=RCIM_Cubic, leaveTangent=-0.4506, tangentMode=RCTM_User, time=1.6334, value=0}
        ]}}}
     CurveIndex=$Emitter.Age
    )
    Scale RGB=&Lerp_Vector(
     A=&MakeVectorFromLinearColorRGB(Color=&ColorFromCurve(ColorCurve=DataInterface() {propertyValues {
         AlphaCurve {keys [<at 0.4726 1>,<at 1 0>]}
         BlueCurve {keys [<at 0.0112 4.9932>,<at 0.106 3>,<at 0.5744 1>,<at 0.8933 0.8464>]}
         GreenCurve {keys [<at 0.0112 0.5218>,<at 0.106 0>,<at 0.5744 0.0001>,<at 0.8933 0.0294>]}
         RedCurve {keys [<at 0.0112 5>,<at 0.106 2.9419>,<at 0.5744 0.5227>,<at 0.8933 0.1145>]}
        }}))
     B=&MakeVectorFromLinearColorRGB(Color=&ColorFromCurve(ColorCurve=DataInterface() {propertyValues {
         AlphaCurve {keys []}
         BlueCurve {keys [<at 0.3582 1>,<at 1 1 RCIM_Cubic>]}
         GreenCurve {keys [<at 0.3582 0>,<at 1 0.0292 RCIM_Cubic>]}
         RedCurve {keys [<at 0.3582 0.3076>,<at 1 0.0488 RCIM_Cubic>]}
        }}))
     Alpha=$Particles.SourceParticleAge
    )
    ScaleA=true
    ScaleRGB=true
   )
  ]
  rendererClasses [&NiagaraRibbonRendererProperties]
  renderers [{rendererClass=&NiagaraRibbonRendererProperties}]
 }
}
summary {
 systemName=NS_Emote_MediCrow_Hands
 emitters [
  {emitterName=WispyTrail, rendererClasses [&NiagaraRibbonRendererProperties]}
  {emitterName=Aura, rendererClasses [&NiagaraSpriteRendererProperties]}
 ]
 userVariables [
  User.bIsFrontEnd {type=NiagaraBool}
  User.bIsFrontEndPreview {type=NiagaraBool}
  User.ParaGlideLeanAlpha {type=NiagaraFloat, defaultValue=1, description=""}
  User.RibbonIDOffset {type=NiagaraInt32}
  User.Variant {type=NiagaraFloat, defaultValue=0, description=""}
 ]
}
