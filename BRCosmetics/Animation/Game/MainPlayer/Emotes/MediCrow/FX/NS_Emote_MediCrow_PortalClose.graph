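# Niagara effect for the MediCrow emote's portal closing: a short one-shot burst of feathers, flash, smoke and dust.
# Feathers are tumbling mesh particles with drag and drift, while a brief purple flare with glow, flipbook smoke and expanding dust use sprites.
# Time-of-day functions adjust flare, smoke and dust alpha, plus flare and smoke color.
# A Shockwave emitter has no renderers, and front-end user variables are declared without visible use here.
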NiagaraSystem /BRCosmetics/Animation/Game/MainPlayer/Emotes/MediCrow/FX/NS_Emote_MediCrow_PortalClose
dependencies {
 usedDataInterfaces [
  &NiagaraDataInterfaceCurve
  &NiagaraDataInterfaceVector2DCurve
  &NiagaraDataInterfaceSpriteRendererInfo
  &NiagaraDataInterfaceColorCurve
 ]
 usedDynamicInputs [&RandomRangeVector, &RandomRangeVector2D]
 usedModules [&SystemState]
}
emitters {
 Feathers {
  emitterUpdateScript [
   &EmitterState(Life Cycle Mode=::NewEnumerator1, Loop Behavior=::NewEnumerator1, Loop Delay=0.1, Loop Duration=1, UseLoopDelay=true)
   &SpawnBurst_Instantaneous(Spawn Count=30, Use Spawn Probability=false, UseLoopCountLimit=false)
  ]
  particleSpawnScript [
   &InitializeParticle(
    Color=<rgba 0.8894 0.7274 1 1>
    Lifetime Max=0.5
    Lifetime Min=0.9
    Lifetime Mode=::NewEnumerator1
    Mesh Scale Mode=::NewEnumerator3
    Mesh Uniform Scale=&Multiply_Float(A=&RandomRangeFloat(Minimum=0.1, Maximum=0.2), B=6)
    UsePositionOffset=false
    Write Material Random=true
   )
   &InitialMeshOrientation(Use Rotation Vector=false)
   &ShapeLocation(MaskSpawnGroup=false, Override Seed=false, Sphere Radius=30, UseOwnerScale=true)
   &AddVelocity(
    Cone Angle=160
    Cone Axis=<0 0 1>
    Module.UseConeFalloff=false
    Module.UseVelDistribution=false
    Override Seed=false
    Velocity Mode=::NewEnumerator2
    Velocity Speed=&RandomRangeFloat(Minimum=300, Maximum=200)
   )
  ]
  particleUpdateScript [
   &Drag(
    Drag=&FloatFromCurve(
     FloatCurve=DataInterface() {propertyValues {Curve {keys [
         {interpMode=RCIM_Cubic, tangentMode=RCTM_User, time=0, value=0}
         {arriveTangent=36.9069, interpMode=RCIM_Cubic, leaveTangent=36.9069, tangentMode=RCTM_User, time=0.0782, value=1}
        ]}}}
     Scale Curve=5
    )
    Use Linear Drag=true
    Use Rotational Drag=false
   )
   &GravityForce(Gravity=<0 0 -25>)
   &AccelerationForce(Acceleration=<0 20 0>)
   &SolveForcesAndVelocity(Clamp Velocity=false, Limit Acceleration=false)
   &ParticleState()
   &DynamicMaterialParameters(
    Index 0 Param 1=&RandomRangeFloat(Minimum=0, Maximum=0.1)
    Index 0 Param 2=&FloatFromCurve(FloatCurve=DataInterface() {propertyValues {Curve {keys [<at 0.6 1>,<at 1 0>]}}})
    Index 0 Param 3=&FloatFromCurve(FloatCurve=DataInterface() {propertyValues {Curve {keys [<at 0 1>,<at 0.5 10>]}}})
    Index 0 Param 4=&RandomRangeFloat(Minimum=0, Maximum=0.5)
    Module.Index0Float0Write=true
    Module.Index0Float1Write=true
    Module.Index0Float2Write=true
    Module.Index0Float3Write=true
   )
   &ScaleColor(Scale RGB=&Multiply_VectorByFloat(Vector=<1 1 1>, Float=1.3), ScaleA=true, ScaleRGB=true)
   &UpdateMeshOrientation(Rotation Vector=&VectorFromFloat(Value=&RandomRangeFloat(Minimum=-1, Maximum=1)))
   &ScaleMeshSize(Scale Factor=&Multiply_VectorByFloat(
     Vector=<1 1 1>
     Float=&FloatFromCurve(FloatCurve=DataInterface() {propertyValues {Curve {keys [
          {arriveTangent=4.1785, interpMode=RCIM_Cubic, leaveTangent=4.1785, tangentMode=RCTM_User, time=0, value=0.5861}
          {arriveTangent=0.9995, interpMode=RCIM_Cubic, leaveTangent=-0.1394, tangentMode=RCTM_Break, time=0.15, value=1.3354}
          {arriveTangent=-1.5711, interpMode=RCIM_Cubic, leaveTangent=-1.5711, time=0.5, value=1}
          {interpMode=RCIM_Cubic, leaveTangent=-1.5153, time=1, value=0}
         ]}}})
    )) {module=ScaleMeshSize001}
   &CameraOffset(Camera Offset Amount=15)
  ]
  rendererClasses [&NiagaraMeshRendererProperties]
  renderers [
   {rendererClass=&NiagaraMeshRendererProperties, rendererIndex=0}
   {rendererClass=&NiagaraMeshRendererProperties, rendererIndex=1}
  ]
 }
 Flash_Flare {
  emitterUpdateScript [
   &EmitterState(
    Inactive Response=::NewEnumerator1
    Life Cycle Mode=::NewEnumerator1
    Loop Behavior=::NewEnumerator1
    Loop Delay=0.01
    Loop Duration=1
    UseLoopDelay=false
   )
   &SpawnBurst_Instantaneous(Use Spawn Probability=false, UseLoopCountLimit=false)
  ]
  particleSpawnScript [
   &InitializeParticle(
    Color=<rgba 0.9763 0 3 1>
    Lifetime=0.2
    Position Offset=<0 0 5>
    Sprite Size=<200 200>
    Sprite Size Mode=::NewEnumerator4
    UsePositionOffset=true
    Write Material Random=true
   )
   &CameraOffset(Camera Offset Amount=10)
   &DynamicMaterialParameters(
    Index 0 Param 1=0.2
    Index 0 Param 2=1
    Index 3 Param 1=1.5
    Index 3 Param 2=1
    Module.Index0Float0Write=true
    Module.Index0Float1Write=true
    Module.Index0Float2Write=true
    Module.Index0Float3Write=true
    Module.Index3Float0Write=true
    Module.Index3Float1Write=true
    Module.Index3Float2Write=true
    Module.Index3Float3Write=true
    Write Parameter Index 3=::NewEnumerator1
   )
   /SetVariables(Particles.GlowCameraOffset=30, Particles.GlowSize=<100 100>, Particles.GlowSpriteFacing:Vector3f=<0 0 1>) {isSetParameters=true}
  ]
  particleUpdateScript [
   &ParticleState()
   &ScaleSpriteSize(
    Non-Uniform Curve Sprite Scale=DataInterface() {propertyValues {
      XCurve {keys [{tangentMode=RCTM_User, time=0, value=0.5}, <at 0.2 1.075>, <at 0.4 0.1>, <at 0.8 0.1>, {tangentMode=RCTM_User, time=1, value=0}]}
      YCurve {keys [{tangentMode=RCTM_User, time=0, value=0.5}, <at 0.1 1.025>, {tangentMode=RCTM_User, time=0.9998, value=1.258}]}
     }}
    Scale Sprite Size Mode=::NewEnumerator3
   )
   &SolveForcesAndVelocity(Clamp Velocity=false, Limit Acceleration=false)
   &ScaleColor(
    Scale Alpha=&FNTimeOfDayScalar(Night Scale=0.5, Morning Scale=1, Evening Scale=0.5)
    Scale RGB=&FN_Select_Vector_from_Time_Of_Day(Morning=<1 1 1>, Day=<1 1 1>, Evening=<0.8 0.8 0.8>, Night=<0.8 0.8 0.8>)
    ScaleA=true
    ScaleRGB=true
   )
   /SetVariables(Particles.FlarePivotOffset:Vector2f=&Vector2DFromCurve(
     Vector2Curve=DataInterface() {propertyValues {XCurve {keys [<at 0 1>,<at 1 1>]}, YCurve {keys [<at 0 0.5>,<at 1 0.5458>]}}}
     ScaleCurve=<0.5 1>
    )) {isSetParameters=true}
   /SetVariables(Particles.GlowColor=&MakeLinearColor(
     R:NiagaraFloat=0.223
     G:NiagaraFloat=0
     B:NiagaraFloat=1
     A:NiagaraFloat=&FloatFromCurve(
      FloatCurve=DataInterface() {propertyValues {Curve {keys [
          {arriveTangent=5.0727, interpMode=RCIM_Cubic, leaveTangent=5.0727, tangentMode=RCTM_User, time=0, value=0.1641}
          {
           arriveTangent=0.1472
           arriveTangentWeight=0.3592
           interpMode=RCIM_Cubic
           leaveTangent=0.1472
           leaveTangentWeight=0.3592
           tangentMode=RCTM_User
           tangentWeightMode=RCTWM_WeightedBoth
           time=0.5103
           value=1
          }
          {tangentMode=RCTM_User, time=1, value=0}
         ]}}}
      Scale Curve=0.751
     )
    )) {isSetParameters=true}
  ]
  rendererClasses [&NiagaraSpriteRendererProperties]
  renderers [
   {rendererClass=&NiagaraSpriteRendererProperties}
   {rendererClass=&NiagaraSpriteRendererProperties, rendererIndex=1}
  ]
 }
 Shockwave {rendererClasses [], renderers []}
 SmokeBurst {
  emitterUpdateScript [
   &EmitterState(Life Cycle Mode=::NewEnumerator1, Loop Behavior=::NewEnumerator1, Loop Delay=2.6, Loop Duration=2, UseLoopDelay=false)
   &SpawnBurst_Instantaneous(Spawn Count=5, Use Spawn Probability=false, UseLoopCountLimit=false)
  ]
  particleSpawnScript [
   &InitializeParticle(
    Lifetime Max=0.7
    Lifetime Min=0.8
    Lifetime Mode=::NewEnumerator1
    Sprite Rotation Mode=::NewEnumerator2
    Sprite Size Mode=::NewEnumerator1
    Uniform Sprite Size Max=150
    Uniform Sprite Size Min=50
    UsePositionOffset=false
    Write Material Random=true
   )
   &AddVelocity(
    Cone Angle=90
    Cone Axis=<0 0 1>
    Module.UseConeFalloff=false
    Module.UseVelDistribution=false
    Override Seed=false
    Velocity Mode=::NewEnumerator2
    Velocity Speed=&RandomRangeFloat(Minimum=0, Maximum=100)
   )
  ]
  particleUpdateScript [
   &ParticleState()
   &SpriteRotationRate(Rotation Rate=&RandomRangeFloat(Minimum=-50, Maximum=50))
   &ScaleSpriteSize(
    Scale Sprite Size Mode=::NewEnumerator0
    Uniform Scale Factor=&FloatFromCurve(FloatCurve=DataInterface() {propertyValues {Curve {keys [<at 0 0.8>,<at 1 1.2>]}}}, Scale Curve=1.3)
   )
   &SubUVAnimation(
    End Frame Range Override=20
    "Sprite Renderer "=DataInterface() {propertyValues {SpriteRenderer=&self:SmokeBurst_0.NiagaraSpriteRendererProperties_0}}
    Start Frame Range Override=&RandomRangeInt(Minimum=0, Maximum=3)
    UseEndFrame=true
    UseStartFrame=true
   )
   &ScaleColor(
    Scale Alpha=&FloatFromCurve(
     FloatCurve=DataInterface() {propertyValues {Curve {keys [
         {tangentMode=RCTM_User, time=0, value=1}
         {arriveTangent=-0.0405, interpMode=RCIM_Cubic, leaveTangent=-0.0405, tangentMode=RCTM_User, time=0.7009, value=1}
         {arriveTangent=-8.0089, interpMode=RCIM_Cubic, leaveTangent=-8.0089, tangentMode=RCTM_User, time=1, value=0}
        ]}}}
     Scale Curve=&FNTimeOfDayScalar(Night Scale=2, Morning Scale=5, Day Scale=5, Evening Scale=5)
    )
    Scale RGB=&FN_Select_Vector_from_Time_Of_Day(
     Morning=&MakeVectorFromLinearColorRGB(Color=&ColorFromCurve(ColorCurve=DataInterface() {propertyValues {
         AlphaCurve {keys [<at 0 1 RCIM_Cubic>,<at 1 0 RCIM_Cubic>]}
         BlueCurve {keys [<at 0 0.092 RCIM_Cubic>,<at 1 0.0828 RCIM_Cubic>]}
         GreenCurve {keys [<at 0 0.026 RCIM_Cubic>,<at 1 0.0449 RCIM_Cubic>]}
         RedCurve {keys [<at 0 0.0375 RCIM_Cubic>,<at 1 0.0376 RCIM_Cubic>]}
        }}))
     Day=&MakeVectorFromLinearColorRGB(Color=&ColorFromCurve(ColorCurve=DataInterface() {propertyValues {
         AlphaCurve {keys [<at 0 1 RCIM_Cubic>,<at 1 0 RCIM_Cubic>]}
         BlueCurve {keys [<at 0 0.092 RCIM_Cubic>,<at 1 0.0828 RCIM_Cubic>]}
         GreenCurve {keys [<at 0 0.026 RCIM_Cubic>,<at 1 0.0449 RCIM_Cubic>]}
         RedCurve {keys [<at 0 0.0375 RCIM_Cubic>,<at 1 0.0376 RCIM_Cubic>]}
        }}))
     Evening=&MakeVectorFromLinearColorRGB(Color=&ColorFromCurve(ColorCurve=DataInterface() {propertyValues {
         AlphaCurve {keys [<at 0 1 RCIM_Cubic>,<at 1 0 RCIM_Cubic>]}
         BlueCurve {keys [<at 0 0.092 RCIM_Cubic>,<at 1 0.0828 RCIM_Cubic>]}
         GreenCurve {keys [<at 0 0.026 RCIM_Cubic>,<at 1 0.0449 RCIM_Cubic>]}
         RedCurve {keys [<at 0 0.0375 RCIM_Cubic>,<at 1 0.0376 RCIM_Cubic>]}
        }}))
     Night=&MakeVectorFromLinearColorRGB(Color=&ColorFromCurve(ColorCurve=DataInterface() {propertyValues {
         AlphaCurve {keys [<at 0 1 RCIM_Cubic>,<at 1 0 RCIM_Cubic>]}
         BlueCurve {keys [<at 0 0.092 RCIM_Cubic>,<at 1 0.1337 RCIM_Cubic>]}
         GreenCurve {keys [<at 0 0.0153 RCIM_Cubic>,<at 1 0.0168 RCIM_Cubic>]}
         RedCurve {keys [<at 0 0.015 RCIM_Cubic>,<at 1 0.0169 RCIM_Cubic>]}
        }}))
    )
    ScaleA=true
    ScaleRGB=true
   )
   &GravityForce(Gravity=<0 0 25>)
   &Drag(Drag=3, Use Linear Drag=true, Use Rotational Drag=true)
   &SolveForcesAndVelocity(Clamp Velocity=false, Limit Acceleration=false)
   &CameraOffset(Camera Offset Amount=15)
  ]
  rendererClasses [&NiagaraSpriteRendererProperties]
  renderers [{rendererClass=&NiagaraSpriteRendererProperties}]
 }
 SpawnDust {
  emitterUpdateScript [
   &EmitterState(Life Cycle Mode=::NewEnumerator1, Loop Behavior=::NewEnumerator1, Loop Delay=0.1, Loop Duration=1, UseLoopDelay=true)
   &SpawnBurst_Instantaneous(Spawn Count=10, Use Spawn Probability=false, UseLoopCountLimit=false)
  ]
  particleSpawnScript [
   &InitializeParticle(
    Color Minimum=<rgba 0.2587 0.2587 0.2587 1>
    Color Mode=::NewEnumerator2
    Lifetime Max=0.8
    Lifetime Min=0.6
    Lifetime Mode=::NewEnumerator1
    Sprite Rotation Mode=::NewEnumerator2
    Sprite Size Mode=::NewEnumerator1
    Uniform Sprite Size Max=60
    Uniform Sprite Size Min=40
    UsePositionOffset=false
    Write Material Random=true
   )
   &ShapeLocation(MaskSpawnGroup=false, Override Seed=false, Sphere Radius=15, UseOwnerScale=true)
   &CameraOffset(Camera Offset Amount=&RandomRangeFloat(Minimum=0, Maximum=10))
   &AddVelocity(
    Module.Use Falloff Distance=false
    Override Seed=false
    Velocity Mode=::NewEnumerator1
    Velocity Speed=&RandomRangeFloat(Minimum=60, Maximum=30)
   )
  ]
  particleUpdateScript [
   &ParticleState()
   &SpriteRotationRate(Rotation Rate=&RandomRangeFloat(Minimum=-150, Maximum=150))
   &ScaleSpriteSize(Uniform Curve Sprite Scale=DataInterface() {propertyValues {Curve {keys [
        {arriveTangent=1.6237, interpMode=RCIM_Cubic, leaveTangent=1.6237, tangentMode=RCTM_User, time=0, value=1}
        {arriveTangent=0.637, tangentMode=RCTM_User, time=1, value=1.7}
       ]}}})
   &Drag(
    Drag=&FloatFromCurve(FloatCurve=DataInterface() {propertyValues {Curve {keys [{interpMode=RCIM_Cubic, tangentMode=RCTM_User, time=0, value=0}, {interpMode=RCIM_Cubic, tangentMode=RCTM_User, time=0.4943, value=1}]}}})
    Use Linear Drag=true
    Use Rotational Drag=true
   )
   &SolveForcesAndVelocity(Clamp Velocity=false, Limit Acceleration=false)
   &ScaleColor(
    Scale Alpha=&FloatFromCurve(
     FloatCurve=DataInterface() {propertyValues {Curve {keys [
         {interpMode=RCIM_Cubic, tangentMode=RCTM_User, time=0, value=0.8902}
         {arriveTangent=-2.0189, interpMode=RCIM_Cubic, leaveTangent=-2.1015, tangentMode=RCTM_Break, time=0.2, value=1.8153}
         {arriveTangent=-1.1521, interpMode=RCIM_Cubic, tangentMode=RCTM_Break, time=1, value=0}
        ]}}}
     Scale Curve=&FloatFromCurve(
      FloatCurve=DataInterface() {propertyValues {Curve {keys [{interpMode=RCIM_Cubic, tangentMode=RCTM_User, time=4.5123, value=1}, {interpMode=RCIM_Cubic, tangentMode=RCTM_User, time=4.95, value=0}]}}}
      Scale Curve=&FNTimeOfDayScalar(Night Scale=0.6, Morning Scale=0.4, Day Scale=0.3, Evening Scale=0.4)
     )
    )
    Scale RGB=&MakeVectorFromLinearColorRGB(Color=&ColorFromCurve(ColorCurve=DataInterface() {propertyValues {
        AlphaCurve {keys [<at 0 1 RCIM_Cubic>]}
        BlueCurve {keys [<at 0 0.1314 RCIM_Cubic>,<at 1 0.0295 RCIM_Cubic>]}
        GreenCurve {keys [<at 0 0.0004 RCIM_Cubic>,<at 1 0.0005 RCIM_Cubic>]}
        RedCurve {keys [<at 0 0.0358 RCIM_Cubic>,<at 1 0.01 RCIM_Cubic>]}
       }}))
    ScaleA=true
    ScaleRGB=true
   )
  ]
  rendererClasses [&NiagaraSpriteRendererProperties]
  renderers [{rendererClass=&NiagaraSpriteRendererProperties}]
 }
}
summary {
 systemName=NS_Emote_MediCrow_PortalClose
 emitters [
  {emitterName=Feathers, rendererClasses [&NiagaraMeshRendererProperties]}
  {emitterName=SmokeBurst, rendererClasses [&NiagaraSpriteRendererProperties]}
  {emitterName=SpawnDust, rendererClasses [&NiagaraSpriteRendererProperties]}
  {emitterName=Flash_Flare, rendererClasses [&NiagaraSpriteRendererProperties]}
  {emitterName=Shockwave, rendererClasses []}
 ]
 userVariables [
  User.bIsFrontEnd {type=NiagaraBool}
  User.bIsFrontEndPreview {type=NiagaraBool}
  User.FrontEndScale {type=NiagaraFloat, defaultValue=0.5, description=""}
 ]
}
