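# Niagara system for the MediCrow player emote that launches a short flock of glowing purple crows with trails and death bursts.
# Crows combine sprite and mesh renderers, steered by a point force, line attraction and curve-driven drag.
# WispyTrail ribbons follow each crow via particle attribute reads, tinted by source particle age.
# Feather meshes, smoke and flame sprites also sample Crows, rejecting all but particles near the end of System.CrowLifetime so they burst as crows expire.
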NiagaraSystem /BRCosmetics/Animation/Game/MainPlayer/Emotes/MediCrow/FX/NS_Emote_Medicrow_Crows
dependencies {
 usedDataInterfaces [
  &NiagaraDataInterfaceCurve
  &NiagaraDataInterfaceColorCurve
  &NiagaraDataInterfaceParticleRead
  &NiagaraDataInterfaceVector2DCurve
  &NiagaraDataInterfaceSpriteRendererInfo
  &NiagaraDataInterfaceArrayFloat
 ]
 usedDynamicInputs [
  &Vector2DFromCurve
  &RandomRangeVector2D
  &SelectFloatFromArray
 ]
 usedModules [
  &self:SystemSpawnScript.SystemScriptSource.SystemScriptGraph.NiagaraNodeAssignment_0.SetVariables_#1
  &SystemState
  &self:SystemSpawnScript.SystemScriptSource.SystemScriptGraph.NiagaraNodeAssignment_1.SetVariables_#2
 ]
}
emitters {
 Crows {
  emitterUpdateScript [
   &EmitterState(Life Cycle Mode=::NewEnumerator1, Loop Behavior=::NewEnumerator1, Loop Duration=3, UseLoopDelay=false)
   &SpawnRate(
    SpawnRate=&FloatFromCurve(
     FloatCurve=DataInterface() {propertyValues {Curve {keys [
         {arriveTangent=-0.2612, interpMode=RCIM_Cubic, leaveTangent=-0.2612, tangentMode=RCTM_User, time=0.8001, value=1}
         {arriveTangent=-12.401, interpMode=RCIM_Cubic, leaveTangent=-12.401, tangentMode=RCTM_User, time=1, value=0}
        ]}}}
     CurveIndex=$Emitter.Age
     Scale Curve=30
    )
    Use Spawn Probability=false
   )
   &SpawnBurst_Instantaneous(Spawn Count=3, Use Spawn Probability=false, UseLoopCountLimit=false)
  ]
  particleSpawnScript [
   &PartitionParticles()
   &InitializeParticle(
    Lifetime=$System.CrowLifetime
    Mesh Scale Mode=::NewEnumerator1
    Mesh Uniform Scale Max=0.25
    Mesh Uniform Scale Min=0.15
    Sprite Size=&Multiply_Vector2D_ByFloat(Float=80.1583, Vector2D=<1 1>)
    Sprite Size Mode=::NewEnumerator4
    UsePositionOffset=false
    Write Material Random=true
   )
   &ShapeLocation(MaskSpawnGroup=false, Non Uniform Scale=<1 1 0.3>, Override Seed=false, Sphere Radius=70, UseOwnerScale=true)
   &self:PositionMultiplier() {module=PositionMultiplier}
   &PointForce(
    Force Falloff Distance=200
    Force Strength=&Multiply_Float(
     A=600
     B=&FloatFromCurve(
      FloatCurve=DataInterface() {propertyValues {Curve {keys [{interpMode=RCIM_Cubic, tangentMode=RCTM_User, time=0, value=2}, {interpMode=RCIM_Cubic, tangentMode=RCTM_User, time=1, value=1.9858}]}}}
      CurveIndex=$Emitter.Age
     )
    )
    Linear Force Vector:Vector3f=&MakeVector(
     X=&Multiply_Float(A=0.2, B=$Particles.PositionMultiplier)
     Y=&RandomRangeFloat(Minimum=0, Maximum=0.6)
     Z=0.8
    )
    Module.Use Falloff=false
    Module.UseLinearForce=true
   ) {module=PointForce001}
   &ApplyInitialForces()
  ]
  particleUpdateScript [
   &ParticleState()
   /Niagara/Modules/Update/Forces/LineAttractionForce(
    Attraction Strength=2000
    Line End:Vector3f=<70 100 80>
    Line Segment in Localspace=true
    Line Start:Vector3f=<-70 100 80>
    Use Attraction Falloff=false
   )
   &Drag(
    Drag=&FloatFromCurve(
     FloatCurve=DataInterface() {propertyValues {Curve {keys [
         <at 0.0256 0 RCIM_Cubic>
         {arriveTangent=2.0337, interpMode=RCIM_Cubic, leaveTangent=2.0337, tangentMode=RCTM_User, time=0.0977, value=0.1634}
         {arriveTangent=11.3111, interpMode=RCIM_Cubic, leaveTangent=-21.9758, tangentMode=RCTM_Break, time=0.1389, value=0.9809}
         {arriveTangent=-0.2305, interpMode=RCIM_Cubic, leaveTangent=-0.2305, tangentMode=RCTM_User, time=0.2593, value=0.0037}
         <at 1 0 RCIM_Cubic>
        ]}}}
     Scale Curve=&Multiply_Float(
      A=4
      B=&FloatFromCurve(
       FloatCurve=DataInterface() {propertyValues {Curve {keys [{interpMode=RCIM_Cubic, tangentMode=RCTM_User, time=0, value=2}, {interpMode=RCIM_Cubic, tangentMode=RCTM_User, time=1, value=2.0626}]}}}
       CurveIndex=$Emitter.Age
       Scale Curve=3
      )
     )
    )
    Use Linear Drag=true
    Use Rotational Drag=true
   )
   &SolveForcesAndVelocity(Clamp Velocity=false, Limit Acceleration=false)
   /SetVariables(
    Particles.CrowColor:LinearColor=&ColorFromCurve(ColorCurve=DataInterface() {propertyValues {
       AlphaCurve {keys [<at 0 1 RCIM_Cubic>]}
       BlueCurve {keys [<at 0 5>,<at 0.6996 0.3304>]}
       GreenCurve {keys [<at 0 0.0667>,<at 0.6996 0.0896>]}
       RedCurve {keys [<at 0 1.6434>,<at 0.6996 0.0865>]}
      }})
    Particles.CrowDynamicParameters:Vector4f=&MakeVector4(
     X=0
     Y=&RandomRangeFloat(Minimum=3, Maximum=1)
     Z=&RandomRangeFloat(Minimum=0, Maximum=1)
     W=0
    )
   ) {isSetParameters=true}
   /SetVariables(Particles.GlowColor=&ColorFromCurve(ColorCurve=DataInterface() {propertyValues {
       AlphaCurve {keys [<at 0 1 RCIM_Cubic>]}
       BlueCurve {keys [<at 0 5 RCIM_Cubic>,<at 0.6444 1 RCIM_Cubic>]}
       GreenCurve {keys [<at 0 0.1486 RCIM_Cubic>,<at 0.6444 0.0151 RCIM_Cubic>]}
       RedCurve {keys [<at 0 1.7185 RCIM_Cubic>,<at 0.6444 0.3953 RCIM_Cubic>]}
      }})) {isSetParameters=true}
  ]
  rendererClasses [&NiagaraSpriteRendererProperties, &NiagaraMeshRendererProperties]
  renderers [
   {rendererClass=&NiagaraSpriteRendererProperties}
   {rendererClass=&NiagaraMeshRendererProperties, rendererIndex=1}
  ]
 }
 FeatherBurst001 {
  emitterUpdateScript [
   &EmitterState(Life Cycle Mode=::NewEnumerator1, Loop Behavior=::NewEnumerator1, Loop Delay=0.5, Loop Duration=3, UseLoopDelay=false)
   &SpawnParticlesFromOtherEmitter(
    Attribute Reader=DataInterface() {propertyValues {EmitterBinding {bindingMode=Other, emitterName=Crows}}}
    Minimum Age Spawn Threshold=0.295
    Spawn Rate=20
    Use Maximum Age=false
    Use Minimum Age=false
   )
  ]
  particleSpawnScript [
   &InitializeParticle(
    Color=<rgba 0.8894 0.7274 1 1>
    Lifetime Max=0.6
    Lifetime Min=0.9
    Lifetime Mode=::NewEnumerator1
    Mesh Scale Mode=::NewEnumerator3
    Mesh Uniform Scale=&Multiply_Float(A=&RandomRangeFloat(Minimum=0.1, Maximum=0.2), B=6)
    UsePositionOffset=false
    Write Material Random=true
   )
   &SampleParticlesFromOtherEmitter(
    Minimum Sampled Particle Age:NiagaraFloat=&Add_Float(A=$System.CrowLifetime, B=-0.04)
    Module.MaxSampledAgeCutoff=false
    Module.MinSampledAgeCutoff=true
    Reject Particle By Sampled Age=true
   )
   /SetVariables(Particles.Position=&TransformPosition(Position=$Output.SampleParticlesFromOtherEmitter.SampledPosition, Source Space=::World, Destination Space=::Local)) {enabled=false, isSetParameters=true}
   &InitialMeshOrientation(Use Rotation Vector=false)
   &InheritVelocity(Module.Limit Speed=false, Velocity Scale=<0.01 0.01 0.01>)
   &AddVelocity(
    Cone Angle=90
    Cone Axis=<0 1 0>
    Module.UseConeFalloff=false
    Module.UseVelDistribution=false
    Override Seed=false
    Velocity Mode=::NewEnumerator2
    Velocity Speed=&RandomRangeFloat(Minimum=300, Maximum=500)
   )
  ]
  particleUpdateScript [
   &Drag(
    Drag=&FloatFromCurve(
     FloatCurve=DataInterface() {propertyValues {Curve {keys [{interpMode=RCIM_Cubic, tangentMode=RCTM_User, time=0, value=0}, {interpMode=RCIM_Cubic, tangentMode=RCTM_User, time=0.0782, value=1}]}}}
     Scale Curve=5
    )
    Use Linear Drag=true
    Use Rotational Drag=false
   )
   &GravityForce(Gravity=<0 0 -25>)
   &AccelerationForce(Acceleration=<0 20 0>)
   &SolveForcesAndVelocity(Clamp Velocity=false, Limit Acceleration=false)
   &ParticleState()
   &DynamicMaterialParameters(
    Index 0 Param 1=&RandomRangeFloat(Minimum=0, Maximum=0.1)
    Index 0 Param 2=&FloatFromCurve(FloatCurve=DataInterface() {propertyValues {Curve {keys [<at 0.6 1>,<at 1 0>]}}})
    Index 0 Param 3=&FloatFromCurve(FloatCurve=DataInterface() {propertyValues {Curve {keys [<at 0 1>,<at 0.5 10>]}}})
    Index 0 Param 4=&RandomRangeFloat(Minimum=0, Maximum=0.5)
    Module.Index0Float0Write=true
    Module.Index0Float1Write=true
    Module.Index0Float2Write=true
    Module.Index0Float3Write=true
   )
   &ScaleColor(Scale RGB=&Multiply_VectorByFloat(Vector=<1 1 1>, Float=1.3), ScaleA=true, ScaleRGB=true)
   &UpdateMeshOrientation(Rotation Vector=&VectorFromFloat(Value=&RandomRangeFloat(Minimum=-1, Maximum=1)))
   &ScaleMeshSize(Scale Factor=&Multiply_VectorByFloat(
     Vector=<1 1 1>
     Float=&FloatFromCurve(FloatCurve=DataInterface() {propertyValues {Curve {keys [
          {arriveTangent=4.1785, interpMode=RCIM_Cubic, leaveTangent=4.1785, tangentMode=RCTM_User, time=0, value=0.5861}
          {arriveTangent=0.9995, interpMode=RCIM_Cubic, leaveTangent=-0.1394, tangentMode=RCTM_Break, time=0.15, value=1.3354}
          {arriveTangent=-1.5711, interpMode=RCIM_Cubic, leaveTangent=-1.5711, time=0.5, value=1}
          {interpMode=RCIM_Cubic, leaveTangent=-1.5153, time=1, value=0}
         ]}}})
    )) {module=ScaleMeshSize001}
   &CameraOffset(Camera Offset Amount=15)
  ]
  rendererClasses [&NiagaraMeshRendererProperties]
  renderers [
   {rendererClass=&NiagaraMeshRendererProperties, rendererIndex=0}
   {rendererClass=&NiagaraMeshRendererProperties, rendererIndex=1}
  ]
 }
 FlameSprites001 {
  emitterUpdateScript [
   &EmitterState(
    Inactive Response=::NewEnumerator1
    Life Cycle Mode=::NewEnumerator1
    Loop Behavior=::NewEnumerator1
    Loop Delay=1.4
    Loop Duration=3
    UseLoopDelay=false
   )
   &SpawnParticlesFromOtherEmitter(
    Attribute Reader=DataInterface() {propertyValues {EmitterBinding {bindingMode=Other, emitterName=Crows}}}
    Minimum Age Spawn Threshold=0.295
    Spawn Rate=20
    Use Maximum Age=false
    Use Minimum Age=false
   )
  ]
  particleSpawnScript [
   &InitializeParticle(
    Color=<rgba 0.4 0.4 0.4 1>
    Lifetime Max=0.3
    Lifetime Min=0.5
    Lifetime Mode=::NewEnumerator1
    Sprite Rotation Mode=::NewEnumerator2
    Sprite Size Mode=::NewEnumerator1
    Uniform Sprite Size Max=15
    Uniform Sprite Size Min=10
    UsePositionOffset=false
    Write Material Random=true
   ) {module=InitializeParticle001}
   &SampleParticlesFromOtherEmitter(
    Minimum Sampled Particle Age:NiagaraFloat=&Add_Float(A=$System.CrowLifetime, B=-0.04)
    Module.MaxSampledAgeCutoff=false
    Module.MinSampledAgeCutoff=true
    Reject Particle By Sampled Age=true
   )
   &AddVelocity(
    Module.Use Falloff Distance=false
    Origin Offset=$Output.SampleParticlesFromOtherEmitter.SampledPosition
    Override Seed=false
    Velocity Mode=::NewEnumerator1
    Velocity Speed=&RandomRangeFloat(Minimum=200, Maximum=75)
   )
  ]
  particleUpdateScript [
   /SetVariables(Particles.SubImageIndex=&FloatFromCurve(FloatCurve=DataInterface() {propertyValues {Curve {keys [<at 0 5>,<at 1 15>]}}})) {isSetParameters=true}
   &ParticleState()
   &ScaleSpriteSize(Uniform Curve Sprite Scale=DataInterface() {propertyValues {Curve {keys [
        {arriveTangent=-1.829, interpMode=RCIM_Cubic, leaveTangent=-1.829, tangentMode=RCTM_User, time=0, value=1.6}
        {interpMode=RCIM_Cubic, tangentMode=RCTM_User, time=1, value=1}
       ]}}})
   &Drag(Drag=2, Use Linear Drag=true, Use Rotational Drag=false)
   &AccelerationForce(Acceleration=<0 20 0>)
   &SolveForcesAndVelocity(Clamp Velocity=false, Limit Acceleration=false)
   &ScaleColor(
    Scale Alpha=&Multiply_Float(
     A=&FloatFromCurve(FloatCurve=DataInterface() {propertyValues {Curve {keys [
          {arriveTangent=4.9417, interpMode=RCIM_Cubic, leaveTangent=4.9417, tangentMode=RCTM_User, time=0, value=0}
          {interpMode=RCIM_Cubic, tangentMode=RCTM_User, time=0.5, value=1}
          {tangentMode=RCTM_User, time=1, value=0}
         ]}}})
     B=&RandomRangeFloat(Minimum=1, Maximum=2)
    )
    Scale RGB=&Multiply_Vector(
     A=<2 2 2>
     B=&MakeVectorFromLinearColorRGB(Color=&ColorFromCurve(ColorCurve=DataInterface() {propertyValues {
         AlphaCurve {keys []}
         BlueCurve {keys [<at 0 1 RCIM_Cubic>,<at 0.3978 0.0412>]}
         GreenCurve {keys [<at 0 0.1071 RCIM_Cubic>,<at 0.3978 0.005>]}
         RedCurve {keys [<at 0 0.5284 RCIM_Cubic>,<at 0.3978 0.0035>]}
        }}))
    )
    ScaleA=true
    ScaleRGB=true
   ) {module=ScaleColor001}
   &CameraOffset(Camera Offset Amount=25)
  ]
  rendererClasses [&NiagaraSpriteRendererProperties]
  renderers [{rendererClass=&NiagaraSpriteRendererProperties}]
 }
 Smoke001 {
  emitterUpdateScript [
   &EmitterState(Life Cycle Mode=::NewEnumerator1, Loop Behavior=::NewEnumerator1, Loop Delay=0.4, Loop Duration=3, UseLoopDelay=false)
   &SpawnParticlesFromOtherEmitter(
    Attribute Reader=DataInterface() {propertyValues {EmitterBinding {bindingMode=Other, emitterName=Crows}}}
    Minimum Age Spawn Threshold=0.295
    Spawn Rate=20
    Use Maximum Age=false
    Use Minimum Age=false
   )
  ]
  particleSpawnScript [
   &InitializeParticle(
    Lifetime Max=0.6
    Lifetime Min=0.3
    Lifetime Mode=::NewEnumerator1
    Sprite Rotation Mode=::NewEnumerator2
    Sprite Size Mode=::NewEnumerator1
    Uniform Sprite Size Max=50
    Uniform Sprite Size Min=40
    UsePositionOffset=false
    Write Material Random=true
   )
   &SampleParticlesFromOtherEmitter(
    Minimum Sampled Particle Age:NiagaraFloat=&Add_Float(A=$System.CrowLifetime, B=-0.04)
    Module.MaxSampledAgeCutoff=false
    Module.MinSampledAgeCutoff=true
    Reject Particle By Sampled Age=true
   )
   &AddVelocity(
    Cone Angle=90
    Cone Axis=<0 1 0>
    Module.UseConeFalloff=false
    Module.UseVelDistribution=false
    Override Seed=false
    Velocity Mode=::NewEnumerator2
    Velocity Speed=&RandomRangeFloat(Minimum=50, Maximum=200)
   )
  ]
  particleUpdateScript [
   &ParticleState()
   &SpriteRotationRate(Rotation Rate=&RandomRangeFloat(Minimum=-50, Maximum=50))
   &ScaleSpriteSize(
    Scale Sprite Size Mode=::NewEnumerator0
    Uniform Scale Factor=&FloatFromCurve(FloatCurve=DataInterface() {propertyValues {Curve {keys [<at 0 0.8>,<at 1 1.2>]}}}, Scale Curve=1.3)
   )
   &SubUVAnimation(
    End Frame Range Override=30
    "Sprite Renderer "=DataInterface() {propertyValues {SpriteRenderer=&self:Smoke001_0.NiagaraSpriteRendererProperties_0}}
    Start Frame Range Override=&RandomRangeInt(Minimum=0, Maximum=5)
    UseEndFrame=true
    UseStartFrame=true
   )
   &ScaleColor(
    Scale Alpha=&FloatFromCurve(
     FloatCurve=DataInterface() {propertyValues {Curve {keys [
         {tangentMode=RCTM_User, time=0, value=1}
         {arriveTangent=-0.0405, interpMode=RCIM_Cubic, leaveTangent=-0.0405, tangentMode=RCTM_User, time=0.7009, value=1}
         {arriveTangent=-8.0089, interpMode=RCIM_Cubic, leaveTangent=-8.0089, tangentMode=RCTM_User, time=1, value=0}
        ]}}}
     Scale Curve=&RandomRangeFloat(Minimum=0.5, Maximum=1)
    )
    Scale RGB=&MakeVectorFromLinearColorRGB(Color=&ColorFromCurve(ColorCurve=DataInterface() {propertyValues {
        AlphaCurve {keys [<at 0 1 RCIM_Cubic>,<at 1 0 RCIM_Cubic>]}
        BlueCurve {keys [<at 0 0.3513 RCIM_Cubic>,<at 0.0949 0.0272 RCIM_Cubic>]}
        GreenCurve {keys [<at 0 0.0994 RCIM_Cubic>,<at 0.0949 0.0148 RCIM_Cubic>]}
        RedCurve {keys [<at 0 0.143 RCIM_Cubic>,<at 0.0949 0.0124 RCIM_Cubic>]}
       }}))
    ScaleA=true
    ScaleRGB=true
   )
   &GravityForce(Gravity=<0 0 25>)
   &Drag(Use Linear Drag=true, Use Rotational Drag=true)
   &SolveForcesAndVelocity(Clamp Velocity=false, Limit Acceleration=false)
   &CameraOffset(Camera Offset Amount=15)
  ]
  rendererClasses [&NiagaraSpriteRendererProperties]
  renderers [{rendererClass=&NiagaraSpriteRendererProperties}]
 }
 WispyTrail {
  emitterUpdateScript [
   &EmitterState(Life Cycle Mode=::NewEnumerator1, Loop Behavior=::NewEnumerator1, Loop Duration=3, UseLoopDelay=false)
   &SpawnParticlesFromOtherEmitter(
    Attribute Reader=DataInterface() {propertyValues {EmitterBinding {bindingMode=Other, emitterName=Crows}}}
    Spawn Rate=&RandomRangeFloat(Minimum=20, Maximum=40)
    Spawn Rate Per Particle Cap=10000000
    Use Maximum Age=false
    Use Minimum Age=false
   )
  ]
  particleSpawnScript [
   &InitializeParticle(UsePositionOffset=false, Write Material Random=true)
   &SampleParticlesFromOtherEmitter(
    Age Sampling:ENiagara_DirectReadSamplingMode=::NewEnumerator1
    Applied Position Offset:Vector3f=<0 0 -5>
    Applied Velocity Scale:NiagaraFloat=&FloatFromCurve(
     FloatCurve=DataInterface() {propertyValues {Curve {keys [{interpMode=RCIM_Cubic, tangentMode=RCTM_User, time=0, value=0}, {interpMode=RCIM_Cubic, tangentMode=RCTM_User, time=0.4174, value=0.136}]}}}
     CurveIndex=$Emitter.Age
    )
    Lifetime Application:ENiagara_LifetimeInheritanceOptions=::NewEnumerator1
    Lifetime Sampling:ENiagara_DirectReadSamplingMode=::NewEnumerator0
    Mesh Scale Sampling:ENiagara_DirectReadSamplingMode=::NewEnumerator1
    Particle ID Sampling=::NewEnumerator1
    Ribbon UV Distance:ENiagara_DirectReadSamplingMode=::NewEnumerator0
    Sprite Size Sampling:ENiagara_DirectReadSamplingMode=::NewEnumerator1
    Velocity Sampling=::NewEnumerator0
   )
   &InitializeRibbon(
    Color=<rgba 0.6485 0 1 0.9>
    Lifetime=&FloatFromCurve(
     FloatCurve=DataInterface() {propertyValues {Curve {keys [
         {arriveTangent=-3.7746, interpMode=RCIM_Cubic, leaveTangent=-3.7746, tangentMode=RCTM_User, time=0, value=1.27}
         {arriveTangent=-0.2864, interpMode=RCIM_Cubic, leaveTangent=-0.2864, tangentMode=RCTM_User, time=0.353, value=0.6}
        ]}}}
     CurveIndex=$Emitter.Age
    )
    Position Mode:ENiagara_PositionInitializationMode=::NewEnumerator0
    Ribbon Width=1
    Write Material Random=true
   )
   /SetVariables(Particles.RibbonWidth=&Multiply_Float(
     A=&MakeFloatFromVector(Vector=$Output.SampleParticlesFromOtherEmitter.SampledScale, Channel=::NewEnumerator0)
     B=90
    )) {isSetParameters=true}
  ]
  particleUpdateScript [
   &self:ReadRibbonSourceParticleAge_06(PAR=DataInterface() {propertyValues {EmitterBinding {bindingMode=Other, emitterName=Crows}}}) {module=ReadRibbonSourceParticleAge_06}
   &ParticleState()
   &SolveForcesAndVelocity(Clamp Velocity=false, Limit Acceleration=false)
   &DynamicMaterialParameters(
    Index 0 Param 1=1
    Module.Index0Float0Write=true
    Module.Index0Float1Write=false
    Module.Index0Float2Write=true
    Module.Index0Float3Write=true
   ) {module=DynamicMaterialParameters001}
   &ScaleColor(
    Scale Alpha=&FloatFromCurve(FloatCurve=DataInterface() {propertyValues {Curve {keys [
         {tangentMode=RCTM_User, time=0, value=1}
         {arriveTangent=-0.0405, interpMode=RCIM_Cubic, leaveTangent=-0.0405, tangentMode=RCTM_User, time=0.0897, value=1}
         {arriveTangent=-5.9438, interpMode=RCIM_Cubic, leaveTangent=-5.9438, tangentMode=RCTM_User, time=0.3888, value=0}
        ]}}})
    Scale RGB=&Lerp_Vector(
     A=&MakeVectorFromLinearColorRGB(Color=&ColorFromCurve(ColorCurve=DataInterface() {propertyValues {
         AlphaCurve {keys [<at 0.4726 1>,<at 1 0>]}
         BlueCurve {keys [<at 0.0112 4.9932>,<at 0.106 3>,<at 0.5744 1>,<at 0.8933 0.8464>]}
         GreenCurve {keys [<at 0.0112 0.5218>,<at 0.106 0>,<at 0.5744 0.0001>,<at 0.8933 0.0294>]}
         RedCurve {keys [<at 0.0112 5>,<at 0.106 2.9419>,<at 0.5744 0.5227>,<at 0.8933 0.1145>]}
        }}))
     B=&RandomRangeVector(
      Minimum=&MakeVectorFromLinearColorRGB(Color=&ColorFromCurve(ColorCurve=DataInterface() {propertyValues {
          AlphaCurve {keys []}
          BlueCurve {keys [<at 0.3582 1>,<at 1 1 RCIM_Cubic>]}
          GreenCurve {keys [<at 0.3582 0>,<at 1 0.0292 RCIM_Cubic>]}
          RedCurve {keys [<at 0.3582 0.3076>,<at 1 0.0488 RCIM_Cubic>]}
         }}))
      Maximum=&MakeVectorFromLinearColorRGB(Color=&ColorFromCurve(ColorCurve=DataInterface() {propertyValues {
          AlphaCurve {keys []}
          BlueCurve {keys [<at 0.3582 1>,<at 1 1 RCIM_Cubic>]}
          GreenCurve {keys [<at 0.3582 0.0862>,<at 1 0.0045 RCIM_Cubic>]}
          RedCurve {keys [<at 0.3582 0.6811>,<at 1 0.3822 RCIM_Cubic>]}
         }}))
     )
     Alpha=$Particles.SourceParticleAge
    )
    ScaleA=true
    ScaleRGB=true
   )
  ]
  rendererClasses [&NiagaraRibbonRendererProperties]
  renderers [{rendererClass=&NiagaraRibbonRendererProperties}]
 }
}
summary {
 systemName=NS_Emote_Medicrow_Crows
 emitters [
  {emitterName=Crows, rendererClasses [&NiagaraSpriteRendererProperties, &NiagaraMeshRendererProperties]}
  {emitterName=WispyTrail, rendererClasses [&NiagaraRibbonRendererProperties]}
  {emitterName=FeatherBurst001, rendererClasses [&NiagaraMeshRendererProperties]}
  {emitterName=Smoke001, rendererClasses [&NiagaraSpriteRendererProperties]}
  {emitterName=FlameSprites001, rendererClasses [&NiagaraSpriteRendererProperties]}
 ]
 userVariables [
  User.bIsFrontEnd {type=NiagaraBool}
  User.bIsFrontEndPreview {type=NiagaraBool}
  User.FrontEndScale {type=NiagaraFloat, defaultValue=0, description=""}
 ]
}
